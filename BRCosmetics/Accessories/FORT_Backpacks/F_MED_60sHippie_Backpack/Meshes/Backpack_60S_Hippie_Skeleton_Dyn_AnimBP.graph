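# Animation Blueprint for the 60s Hippie backpack cosmetic that adds physics-driven secondary motion to its two maracas and tie bones.
# From the reference pose, a chain of AnimDynamics nodes swings each maraca and the central tie under overridden gravity, angular limits and heavy damping.
# The left and right tie bones are constrained to the simulated central tie, then get their own simulation with mirrored side limits.
# The event graph has no logic.
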AnimBlueprint /BRCosmetics/Accessories/FORT_Backpacks/F_MED_60sHippie_Backpack/Meshes/Backpack_60S_Hippie_Skeleton_Dyn_AnimBP
target_skeleton=&/Backpack_60S_Hippie_Skeleton
graphs [
 AnimGraph {nodes [
   n1 = Anim Dynamics - Bone: dyn_lMaraca(ComponentPose=Local To Component(LocalPose=Local Space Ref Pose()), Alpha=1, ExternalForce="0,0,0") {
    angularDampingOverride=0.998
    bGravityOverrideInSimSpace=true
    bOverrideAngularDamping=true
    bUseGravityOverride=true
    bUsePlanarLimit=false
    bUseSphericalLimits=true
    alphaBoolBlend {blendOption=Linear}
    boundBone {boneName=dyn_lMaraca}
    componentAppliedLinearAccClamp=<3000 3000 3000>
    componentLinearAccScale=<0.1 0.2 0.1>
    componentLinearVelScale=<0.1 0.1 0.1>
    gravityOverride {z=-980}
    physicsBodyDefinitions [{
      sphereCollisionRadius=0
      boundBone {boneName=dyn_lMaraca}
      boxExtents=<15 15 10>
      constraintSetup {angularLimitsMax=<5 60 5>, angularLimitsMin=<-5 -60 -5>, angularTarget=<0 0 0>}
      localJointOffset=<0 0 -5>
     }]
   }
   n2 = Anim Dynamics - Bone: dyn_rMaraca(ComponentPose=@n1, Alpha=1, ExternalForce="0,0,0") {
    angularDampingOverride=0.998
    bGravityOverrideInSimSpace=true
    bOverrideAngularDamping=true
    bUseGravityOverride=true
    bUsePlanarLimit=false
    bUseSphericalLimits=true
    alphaBoolBlend {blendOption=Linear}
    boundBone {boneName=dyn_rMaraca}
    componentAppliedLinearAccClamp=<3000 3000 3000>
    componentLinearAccScale=<0.1 0.2 0.1>
    componentLinearVelScale=<0.1 0.1 0.1>
    gravityOverride {z=-980}
    physicsBodyDefinitions [{
      sphereCollisionRadius=0
      boundBone {boneName=dyn_rMaraca}
      boxExtents=<15 15 10>
      constraintSetup {angularLimitsMax=<5 60 5>, angularLimitsMin=<-5 -60 -5>, angularTarget=<0 0 0>}
      localJointOffset=<0 0 -3>
     }]
   }
   n3 = Anim Dynamics - Bone: dyn_tie(ComponentPose=@n2, Alpha=1, ExternalForce="0,0,0") {
    angularDampingOverride=0.999
    bGravityOverrideInSimSpace=true
    bOverrideAngularDamping=true
    bUseGravityOverride=true
    alphaBoolBlend {blendOption=Linear}
    boundBone {boneName=dyn_tie}
    componentAppliedLinearAccClamp=<3000 3000 3000>
    componentLinearAccScale=<0.1 0.2 0.2>
    componentLinearVelScale=<0.1 0.1 0.1>
    gravityOverride {z=-980}
    physicsBodyDefinitions [{
      sphereCollisionRadius=0
      boundBone {boneName=dyn_tie}
      boxExtents=<12 10 6>
      constraintSetup {twistAxis=AxisZ, angularLimitsMax=<0 10 0>, angularLimitsMin=<-60 -10 0>, angularTarget=<0 0 0>}
      localJointOffset=<0 0 -3>
     }]
   }
   n4 = "Constraint - dyn_rtie (P)"(ComponentPose=@n3, Alpha=1, ConstraintWeights_0=1) {
    alphaBoolBlend {blendOption=Linear}
    boneToModify {boneName=dyn_rtie}
    constraintSetup [{targetBone {boneName=dyn_tie}}]
    constraintWeights [1]
   }
   n5 = "Constraint - dyn_ltie (P)"(ComponentPose=@n4, Alpha=1, ConstraintWeights_0=1) {
    alphaBoolBlend {blendOption=Linear}
    boneToModify {boneName=dyn_ltie}
    constraintSetup [{targetBone {boneName=dyn_tie}}]
    constraintWeights [1]
   }
   n6 = Anim Dynamics - Bone: dyn_rtie(ComponentPose=@n5, Alpha=1, ExternalForce="0,0,0") {
    angularDampingOverride=0.997
    bGravityOverrideInSimSpace=true
    bOverrideAngularDamping=true
    bUseGravityOverride=true
    alphaBoolBlend {blendOption=Linear}
    boundBone {boneName=dyn_rtie}
    componentAppliedLinearAccClamp=<3000 3000 3000>
    componentLinearAccScale=<0.3 0.25 0.25>
    componentLinearVelScale=<0.2 0.2 0.2>
    gravityOverride {z=-980}
    physicsBodyDefinitions [{
      sphereCollisionRadius=0
      boundBone {boneName=dyn_rtie}
      boxExtents=<12 10 6>
      constraintSetup {twistAxis=AxisZ, angularLimitsMax=<2 30 0>, angularLimitsMin=<-60 -15 0>, angularTarget=<0 0 0>}
      localJointOffset=<0 0 -2>
     }]
   }
   n7 = Anim Dynamics - Bone: dyn_ltie(ComponentPose=@n6, Alpha=1, ExternalForce="0,0,0") {
    angularDampingOverride=0.997
    bGravityOverrideInSimSpace=true
    bOverrideAngularDamping=true
    bUseGravityOverride=true
    alphaBoolBlend {blendOption=Linear}
    boundBone {boneName=dyn_ltie}
    componentAppliedLinearAccClamp=<3000 3000 3000>
    componentLinearAccScale=<0.3 0.25 0.25>
    componentLinearVelScale=<0.2 0.2 0.2>
    gravityOverride {z=-980}
    physicsBodyDefinitions [{
      sphereCollisionRadius=0
      boundBone {boneName=dyn_ltie}
      boxExtents=<12 10 6>
      constraintSetup {twistAxis=AxisZ, angularLimitsMax=<2 15 0>, angularLimitsMin=<-60 -30 0>, angularTarget=<0 0 0>}
      localJointOffset=<0 0 -3>
     }]
   }
   Output Pose(Result=Component To Local(ComponentPose=@n7)) {layerGroup=None}
  ]}
 EventGraph {nodes [
   TryGetPawnOwner()
   Event BlueprintUpdateAnimation()
  ]}
]
variables []
